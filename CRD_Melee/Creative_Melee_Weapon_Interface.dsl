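; Blueprint interface for Creative melee weapons, declaring getters through which a weapon's tuning data can be queried.
; It covers hit, knockback, block and energy cost data, plus jump attack, sprint attack and secondary ability settings, cannonball impact radius and custom VFX color.
; Hit data takes a gameplay tag, while knockback varies by jump attack and AI target.
; Only signatures with default returns appear, with no logic.
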(fn GetHitData
  (params (Application%20Tag :type Gameplay%20Tag%20Structure))
  (return 0.0 0.0 0.0 0.0 0.0))

(fn GetKnockbackData
  (params (JumpAttack :type Boolean) (TargetAI :type Boolean))
  (return false 0.0 0.0))

(fn GetSecondaryAbility
  (return 0 0.0 0.0 0.0))

(fn GetBlockData
  (return 0.0 0))

(fn GetCannonballImpactRadius
  (return 0.0))

(fn GetCustomVFXColor
  (return false 0 0.0 0.0))

(fn GetEnergyCosts
  (params (NextHit :type Boolean))
  (return 0.0 0.0 0.0 0.0 0.0 0.0))

(fn GetJumpAttackDamage
  (return 0.0))

(fn GetJumpAttackEnabled
  (return false))

(fn GetJumpAttackKnockbackPower
  (return 0.0))

(fn GetSecondaryBreakBlockEnabled
  (return false))

(fn GetSecondaryFeedbackScale
  (return "NewEnumerator0"))

(fn GetSprintAttackDamage
  (return 0.0 0.0))

(fn GetSprintAttackEnabled
  (return false))

(fn GetTargetActionSelection
  (return "NewEnumerator0"))
