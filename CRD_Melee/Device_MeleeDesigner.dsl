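; Creative melee designer device with buttons that spawn a melee item or load a default sword or hammer into its pickup item list.
; Preset buttons async-load the weapon definition, clear the list when only one item is accepted, add it and resave.
; A one-shot binding to the list update then refreshes the pressed preset button.
; Enabled state toggles item templates, and overrides lock scale to 1.
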(event EventInitializeDevice
  ;[0 1 0.1052 0.5] Initialization - Update Button Label from user options.
  (|CallParentFunction)
  (CallFunction|UpdateButtonLabel))

(event OnDeviceButtonClickedDelegate(SpawnWeaponButton)
  (params (InstigatingController :type Fort%20Player%20Controller%20Athena%20Object%20Reference))
  ;[0 0.7912 1 0.5] Item Spawning when pressing the button (parent handles spawning)
  (CallFunction|SpawnMeleeItem))

(event OnEnabledStateChanged(EnabledComponent)
  (params (Enabled :type Boolean))
  (ItemData|SetEnableItemTemplates (Variables|Default|GetPickupItemListComponent) Enabled))

(event Custom|SetWeaponFromButton
  (params (ItemDefinition :type Fort%20World%20Item%20Definition%20Soft%20Object%20Reference))
  ;[1 1 1] Spawn default weapon buttons
  (bind _object (Utilities|AsyncLoadAsset ItemDefinition)
    (:Completed
      (bind _asfort_weapon_item_definition (Utilities|Casting|CastToFortWeaponItemDefinition _object))
      (if (Variables|ItemManagement|GetbOnlyAcceptOneItem)
        (Class|PickupItemListComponent|ClearItemList (Variables|Default|GetPickupItemListComponent))
        (else
          _))
      (ItemData|AddItem (Variables|Default|GetPickupItemListComponent) _asfort_weapon_item_definition 1)
      (CallFunction|Resave))))

(event OnDeviceButtonClickedDelegate(SetSwordButton)
  (params (InstigatingController :type Fort%20Player%20Controller%20Athena%20Object%20Reference))
  ;[1 1 1] Spawn default weapon buttons
  (CallFunction|SetWeaponfromButton (Variables|Default|GetSwordPickup))
  (Default|BindEventtoOnItemListUpdated (Variables|Default|GetPickupItemListComponent) (delegate Custom|OnItemListUpdated_Sword)))

(event Custom|OnItemListUpdated_Sword
  ;[1 1 1] Spawn default weapon buttons
  (Class|FortCreativeDeviceButtonComponent|RequestButtonRefresh (Variables|Default|GetSetSwordButton))
  (Default|UnbindEventfromOnItemListUpdated (Variables|Default|GetPickupItemListComponent) (delegate Custom|OnItemListUpdated_Sword)))

(event OnDeviceButtonClickedDelegate(SetHammerButton)
  (params (InstigatingController :type Fort%20Player%20Controller%20Athena%20Object%20Reference))
  ;[1 1 1] Spawn default weapon buttons
  (CallFunction|SetWeaponfromButton (Variables|Default|GetHammer%20Pickup))
  (Default|BindEventtoOnItemListUpdated (Variables|Default|GetPickupItemListComponent) (delegate Custom|OnItemListUpdated_Hammer)))

(event Custom|OnItemListUpdated_Hammer
  ;[1 1 1] Spawn default weapon buttons
  (Class|FortCreativeDeviceButtonComponent|RequestButtonRefresh (Variables|Default|GetSetHammerButton))
  (Default|UnbindEventfromOnItemListUpdated (Variables|Default|GetPickupItemListComponent) (delegate Custom|OnItemListUpdated_Hammer)))

(fn GetCollisionStaticMesh
  (return "0, 0, 0" (Utilities|ResolveSoftReference (Class|CreativeEditOnlyMeshComponent|GetEditMesh (Variables|Default|GetCreativeEditOnlyMesh)))))

(fn GetMeshComponents
  (return 0))

(fn GetScaleMax
  (params (DefaultScaleMax :type Float%20%28single-precision%29))
  (return 1.0))

(fn GetScaleMin
  (params (DefaultScaleMin :type Float%20%28single-precision%29))
  (return 1.0))

(fn IgnoreWorldGridScaleOverride
  (return false))

(fn UpdateButtonLabel
  (Networking|CanExecuteCosmeticEvents
    (:True
      (CreativeDeviceButtonComponent|SetButtonLabel (Variables|Default|GetSpawnWeaponButton) (Variables|Default|GetButton%20Text%20Spawn%20Item)))))

(fn ConstructionScript)
