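; Parent Blueprint for Creative melee devices that hold dropped melee pickups and can spawn the first stored item above themselves.
; Outside play mode, server-side drop-area hits queue tag-validated pickups, rejecting extras when single-item mode is full.
; List updates rebind the variant notifier and refresh the edit-only mesh's async-loaded icon and rarity colors, with a manual mesh-loaded call working around an editor delegate issue.
; Additions and spawns multicast burst effects.
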(event Custom|RejectItem
  (params (Item :type Actor%20Object%20Reference)))

(event EventInitializeDevice
  ;[0 1 0.2776 0.25] Device Initialization
  (|CallParentFunction)
  (CallFunction|UpdateRarity)
  (CallFunction|UpdateIcon))

(event OnItemListUpdated(PickupItemListComponent)
  ;[0 1 0.2776 0.25] Device Initialization
  (CallFunction|BindNotifiertoNewItem)
  (Networking|CanExecuteCosmeticEvents
    (:True
      (CallFunction|InitializeDevice)
      (CallFunction|UpdateIcon))))

(event OnItemAdded(PickupItemListComponent)
  (params (ItemAdded :type Actor%20Object%20Reference))
  ;[1 1 1] Item dropped in DropArea
  (bind _asfort_pickup (Utilities|Casting|CastToFortPickup ItemAdded))
  (Variables|ItemManagement|SetLastAdded (Class|FortItemFunctionLibrary|BreakFortItemEntry (Pickup|GetPrimaryItemEntry _asfort_pickup)))
  (CallFunction|Resave)
  (CallFunction|MulticastSpawnWeaponBurstEffect (Utilities|Struct|MakeFortBurstEffectData :BurstSounds (Variables|Default|GetAttach%20Burst%20Sounds) :BurstCameraShake "(Shake=None,Scale=1.000000,PlaySpace=CameraLocal,UserPlaySpaceRotation=(Pitch=0.000000,Yaw=0.000000,Roll=0.000000),bAlwaysPlayOnTarget=True,bCalculateUserPlaySpaceRotationFromLocation=False,bCancelOnRemove=False,Falloff=(InnerRadius=0.000000,OuterRadius=0.000000))" :BurstCameraLensEffect "(CameraLensEffect=None,Falloff=(InnerRadius=0.000000,OuterRadius=0.000000),bAlwaysPlayOnTarget=True,bCancelOnRemove=False,bPooled=False)" :BurstForceFeedbackEffect "(ForceFeedbackEffect=None,EffectRadius=100.000000,FarForceFeedbackEffect=None,FarEffectRadius=0.000000,EffectTag=\"\",bAlwaysPlayOnTarget=True)" :BurstDecal "(Condition=(SourceCondition=AnySource,AllowedSurfaces=,DisallowedSurfaces=,ChanceToPlay=1.000000,Significance=High,RequiredDetailMode=0,bRequireVisible=False),Attachment=(SocketName=\"\",AttachType=DoNotAttach,bAttachToWeapon=False,bAttachToHitResult=False,bUseUnsmoothedNetworkPosition=False,bIgnoreLocation=False,bIgnoreScale=True,bIgnoreRotation=False,OverrideLocation=(X=0.000000,Y=0.000000,Z=0.000000),OverrideScale=(X=1.000000,Y=1.000000,Z=1.000000),OverrideRotation=(Pitch=0.000000,Yaw=0.000000,Roll=0.000000)),bOverrideCondition=False,bOverrideAttachment=False,bOverrideFadeOut=False,Decal=None,FadeOutStartDelay=0.000000,FadeOutDuration=0.000000)" :bCancelSoundsOnRemove false :bCancelParticlesOnRemove false)))

(event OnComponentHit(ItemDropArea)
  (params (HitComponent :type Primitive%20Component%20Object%20Reference) (OtherActor :type Actor%20Object%20Reference) (OtherComp :type Primitive%20Component%20Object%20Reference) (NormalImpulse :type Vector :default "0, 0, 0") (Hit :type Hit%20Result%20Structure%20%28by%20ref%29))
  ;[1 1 1] Item dropped in DropArea
  (Networking|SwitchHasAuthority
    (:Authority
      (Utilities|IsValid OtherActor
        (:Is%20Valid
          (CallFunction|HandleItemDropAreaHit OtherActor))))))

(event OnEnabledStateChanged(EnabledComponent)
  (params (Enabled :type Boolean))
  (ItemData|SetEnableItemTemplates (Variables|Default|GetPickupItemListComponent) Enabled))

(event Custom|Multicast_SpawnWeaponBurstEffect
  (params (BurstEffectData :type Fort%20Burst%20Effect%20Data%20Structure))
  ;[1 1 1] Item Spawning
  (GameplayCue|PlayBurstEffect BurstEffectData self (Ability|GameplayCue|MakeGameplayCueParameters :NormalizedMagnitude 0.0 :RawMagnitude 0.0 :Location (Transformation|GetActorLocation) :Normal "0,0,0" :GameplayEffectLevel 1 :AbilityLevel 1 :bReplicateLocationWhenUsingMinimalRepProxy false)))

(event Custom|Spawn%20Melee%20Item
  (bind _pickupitemlistcomponent (Variables|Default|GetPickupItemListComponent))
  ;[1 1 1] Item Spawning
  (Networking|SwitchHasAuthority
    (:Authority
      (if (> (ItemData|GetNumItems _pickupitemlistcomponent) 0)
        (Class|MinigameSpawnerComponent|Spawn (Variables|Default|GetMinigame_Spawner_Component) (ItemData|MakeSpawnParams _pickupitemlistcomponent 0 (Math|Transform|MakeTransform (+ (.location (Transformation|GetActorTransform)) "0, 0,120.000000"))))
        (CallFunction|MulticastSpawnWeaponBurstEffect (Utilities|Struct|MakeFortBurstEffectData :BurstSounds (Variables|Default|GetSpawn%20Burst%20Sounds) :BurstCameraShake "(Shake=None,Scale=1.000000,PlaySpace=CameraLocal,UserPlaySpaceRotation=(Pitch=0.000000,Yaw=0.000000,Roll=0.000000),bAlwaysPlayOnTarget=True,bCalculateUserPlaySpaceRotationFromLocation=False,bCancelOnRemove=False,Falloff=(InnerRadius=0.000000,OuterRadius=0.000000))" :BurstCameraLensEffect "(CameraLensEffect=None,Falloff=(InnerRadius=0.000000,OuterRadius=0.000000),bAlwaysPlayOnTarget=True,bCancelOnRemove=False,bPooled=False)" :BurstForceFeedbackEffect "(ForceFeedbackEffect=None,EffectRadius=100.000000,FarForceFeedbackEffect=None,FarEffectRadius=0.000000,EffectTag=\"\",bAlwaysPlayOnTarget=True)" :BurstDecal "(Condition=(SourceCondition=AnySource,AllowedSurfaces=,DisallowedSurfaces=,ChanceToPlay=1.000000,Significance=High,RequiredDetailMode=0,bRequireVisible=False),Attachment=(SocketName=\"\",AttachType=DoNotAttach,bAttachToWeapon=False,bAttachToHitResult=False,bUseUnsmoothedNetworkPosition=False,bIgnoreLocation=False,bIgnoreScale=True,bIgnoreRotation=False,OverrideLocation=(X=0.000000,Y=0.000000,Z=0.000000),OverrideScale=(X=1.000000,Y=1.000000,Z=1.000000),OverrideRotation=(Pitch=0.000000,Yaw=0.000000,Roll=0.000000)),bOverrideCondition=False,bOverrideAttachment=False,bOverrideFadeOut=False,Decal=None,FadeOutStartDelay=0.000000,FadeOutDuration=0.000000)" :bCancelSoundsOnRemove false :bCancelParticlesOnRemove false))))))

(event Custom|Update%20Icon
  ;[1 0.6967 0] Updating Icon and Rarity.
  (Networking|CanExecuteCosmeticEvents
    (:True
      (Class|CreativeEditOnlyMeshComponent|SetEditMeshandMaterials (Variables|Default|GetCreativeEditOnlyMesh) (Class|CreativeEditOnlyMeshComponent|GetEditMesh (Variables|Default|GetCreativeEditOnlyMesh)) (Utilities|Array|MakeArray (select (> (ItemData|GetNumItems (Variables|Default|GetPickupItemListComponent)) 0) (Variables|Default|GetIcon%20Material) (Variables|Default|GetNo%20Icon%20Material%20Soft%20Ref))) false)
      (if (not (Editor|InEditor))
        (Class|CreativeEditOnlyMeshComponent|ReloadMeshandMaterials (Variables|Default|GetCreativeEditOnlyMesh))
        (else
          (CallFunction|EditorMeshLoaded) ; Looks like there's an issue where ReloadMeshAndMaterials doesn't always fire the delegate in Editor mode,\nso manually calling it here.
        )))))

(event Custom|EditorMeshLoaded
  (bind _pickupitemlistcomponent (Variables|Default|GetPickupItemListComponent))
  (bind _returnvalue (ItemData|GetContainedItems _pickupitemlistcomponent))
  (bind _output (Utilities|Array|Get(acopy) _returnvalue))
  ;[1 0.6967 0] Updating Icon and Rarity.
  (if (> (ItemData|GetNumItems _pickupitemlistcomponent) 0)
    (Utilities|IsValid _output
      (:Is%20Valid
        (Material|ReattachOrCreateMID (Variables|Default|GetCreativeEditOnlyMesh) (Variables|Default|GetPreviewMID))
        (bind _loaded_object (Creative|Loading|AsyncLoadEditorSafe (Item|GetLargePreviewImage _output))
          (:Loaded
            (bind _astexture_2d (Utilities|Casting|CastToTexture2D _loaded_object))
            (Rendering|Material|SetTextureParameterValue (Variables|Default|GetPreviewMID) "ItemIcon" _astexture_2d)
            (bind _loaded_object_2 (Creative|Loading|AsyncLoadEditorSafe (Variables|Default|GetDevice%20Icon%20Override))
              (:Loaded
                (bind _astexture_2d_2 (Utilities|Casting|CastToTexture2D _loaded_object_2))
                (Rendering|Material|SetTextureParameterValue (Variables|Default|GetPreviewMID) "DeviceIcon" _astexture_2d_2)
                (CallFunction|UpdateRarity)))))))
    (else
      (Material|ReattachOrCreateMID (Variables|Default|GetCreativeEditOnlyMesh) (Variables|Default|GetPreviewMID)))))

(event OnMeshandMaterialsLoaded(CreativeEditOnlyMesh)
  (bind _pickupitemlistcomponent (Variables|Default|GetPickupItemListComponent))
  (bind _returnvalue (ItemData|GetContainedItems _pickupitemlistcomponent))
  (bind _output (Utilities|Array|Get(acopy) _returnvalue))
  ;[1 0.6967 0] Updating Icon and Rarity.
  (if (> (ItemData|GetNumItems _pickupitemlistcomponent) 0)
    (Utilities|IsValid _output
      (:Is%20Valid
        (Material|ReattachOrCreateMID (Variables|Default|GetCreativeEditOnlyMesh) (Variables|Default|GetPreviewMID))
        (bind _loaded_object (Creative|Loading|AsyncLoadEditorSafe (Item|GetLargePreviewImage _output))
          (:Loaded
            (bind _astexture_2d (Utilities|Casting|CastToTexture2D _loaded_object))
            (Rendering|Material|SetTextureParameterValue (Variables|Default|GetPreviewMID) "ItemIcon" _astexture_2d)
            (bind _loaded_object_2 (Creative|Loading|AsyncLoadEditorSafe (Variables|Default|GetDevice%20Icon%20Override))
              (:Loaded
                (bind _astexture_2d_2 (Utilities|Casting|CastToTexture2D _loaded_object_2))
                (Rendering|Material|SetTextureParameterValue (Variables|Default|GetPreviewMID) "DeviceIcon" _astexture_2d_2)
                (CallFunction|UpdateRarity)))))))
    (else
      (Material|ReattachOrCreateMID (Variables|Default|GetCreativeEditOnlyMesh) (Variables|Default|GetPreviewMID)))))

(event Custom|Update%20Rarity
  (bind _default_beam_color (Variables|Default|GetDefault%20Beam%20Color))
  (bind _pickupitemlistcomponent (Variables|Default|GetPickupItemListComponent))
  ;[1 0.6967 0] Updating Icon and Rarity.
  (if (> (ItemData|GetNumItems _pickupitemlistcomponent) 0)
    (bind _returnvalue (Fortnite|GetRarityColorPalette (Creative|ItemData|GetRarity (Utilities|Array|Get(acopy) (ItemData|GetContainedItemVariants _pickupitemlistcomponent)))))
    (Material|ReattachOrCreateMID (Variables|Default|GetCreativeEditOnlyMesh) (Variables|Default|GetPreviewMID))
    (bind _previewmid (Variables|Default|GetPreviewMID)
      (:then
        (Rendering|Material|SetVectorParameterValue _previewmid "Color" (Utilities|Struct|BreakFortColorPalette _returnvalue) :!class /Script/Engine.MaterialInstanceDynamic)
        (Maerial|SetColor (Variables|Default|GetCreativeEditOnlyBase) (Variables|Default|GetBase%20Color) (Utilities|Struct|BreakFortColorPalette _returnvalue))
        (Material|SetColor (Variables|Default|GetCreativeEditOnlyBeam) (Utilities|Struct|BreakFortColorPalette _returnvalue)))
      (:else))
    (else
      (Maerial|SetColor (Variables|Default|GetCreativeEditOnlyBase) (Variables|Default|GetBase%20Color) _default_beam_color)
      (Material|SetColor (Variables|Default|GetCreativeEditOnlyBeam) _default_beam_color))))

(fn ActivateDevice
  (|CallParentFunction))

(fn BindNotifierToNewItem
  (bind _pickupitemlistcomponent (Variables|Default|GetPickupItemListComponent))
  (bind _returnvalue (ItemData|GetContainedItemVariants _pickupitemlistcomponent))
  (bind _itemvariantnotifier (Variables|Default|GetItemVariantNotifier))
  (ItemTemplate|UnbindfromAllItemVariants _itemvariantnotifier)
  (if (Utilities|Array|IsValidIndex _returnvalue)
    (ItemTemplate|BindtoItemVariant _itemvariantnotifier (Utilities|Array|Get(acopy) _returnvalue))))

(fn BlueprintOnLocalInteract
  (params (InteractingPawn :type Fort%20Player%20Pawn%20Object%20Reference) (InteractionBeingAttempted :type EInteractionBeingAttempted%20Enum :default "FirstInteraction"))
  (bind _returnvalue (|CallParentFunction InteractingPawn InteractionBeingAttempted))
  (Variables|Default|SetInteracting%20Pawn InteractingPawn)
  (return _returnvalue))

(fn HandleItemDropAreaHit
  (params (OverlappingActor :type Actor%20Object%20Reference))
  (if (not (and (> (ItemData|GetNumItems (Variables|Default|GetPickupItemListComponent)) 0) (Variables|ItemManagement|GetbOnlyAcceptOneItem)))
    (if (not (CallFunction|IsinPlayMode))
      (bind _asfort_pickup_athena (Utilities|Casting|CastToFortPickupAthena OverlappingActor))
      (bind _output_get (LocalVariables|SetPotentialNextFortPickupAthena _asfort_pickup_athena))
      (if (GameplayTags|DoesContainerMatchTagQuery (GameplayTags|GetOwnedGameplayTags (Class|FortItemFunctionLibrary|BreakFortItemEntry (Pickup|GetPrimaryItemEntry _output_get)) :!class /Script/GameplayTags.GameplayTagAssetInterface) (GameplayTags|MakeGameplayTagQuery "(TokenStreamVersion=0,TagDictionary=((TagName=\"MeleeManager.Weapon.Creative.Melee\")),QueryTokenStream=(0,1,2,1,0),UserDescription=\"\",AutoDescription=\" ALL( MeleeManager.Weapon.Creative.Melee )\")"))
        (Utilities|IsPickupValid (LocalVariables|GetPotentialNextFortPickupAthena) (Variables|ItemManagement|GetInvalidPickupGameplayTagContainer)
          (:IsValid
            (Utilities|FortPickupValidforDesigner (LocalVariables|GetPotentialNextFortPickupAthena)
              (:true
                (if (Variables|ItemManagement|GetbOnlyAcceptOneItem)
                  (Class|PickupItemListComponent|ClearItemList (Variables|Default|GetPickupItemListComponent))
                  (else
                    _))
                (Class|PickupItemListComponent|AddItemtoQueue (Variables|Default|GetPickupItemListComponent) (LocalVariables|GetPotentialNextFortPickupAthena))))))))
    (else
      (CallFunction|RejectItem OverlappingActor))))

(fn ConstructionScript)
