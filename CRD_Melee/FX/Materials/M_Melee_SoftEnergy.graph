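# Unlit additive Niagara sprite material for melee FX that draws a soft energy glow with a panning smoky trail.
# A sphere mask on deformed UVs, eased by a sine curve, forms the glow and masks a smoke texture with per-particle random stretch and offset.
# Particle color tints both layers, with optional time-of-day multipliers before emissive adjustments.
# Dynamic parameters drive shape, pan speed and depth fade, which is bypassed on mobile.
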Material /CRD_Melee/FX/Materials/M_Melee_SoftEnergy
graph {
 properties {bUsedWithNiagaraSprites=true, blendMode=BLEND_Additive, pixelDepthOffsetMode=PDOM_Legacy, shadingModel=MSM_Unlit, usedShadingModels=Unlit}
 nodes [
  n1 = MaterialFunctionCall(
   Color_Morning=VectorParameter() {group=TOD, parameterName=Morning, defaultValue=<rgba 1 1 1 1>}.RGB
   Color_Day=VectorParameter() {group=TOD, parameterName=Day, defaultValue=<rgba 1 1 1 1>}.RGB
   Color_Evening=VectorParameter() {group=TOD, parameterName=Evening, defaultValue=<rgba 1 1 1 1>}.RGB
   Color_Night=VectorParameter() {group=TOD, parameterName=Night, defaultValue=<rgba 1 1 1 1>}.RGB
  ) {materialFunction=&Color_Time_of_Day}
  n2 = MaterialFunctionCall(Float2=TextureCoordinate()) {materialFunction=&BreakOutFloat2Components}
  n3 = OneMinus(None=@n2.G)
  n4 = OneMinus(None=DynamicParameter() {defaultValue=<rgba 1 12 0 1>, paramNames [Exponent, Depth Fade, Gradient Deform, Panning Speed]}.Gradient Deform)
  n5 = Constant() {r=0.5}
  n6 = SphereMask(A=AppendVector(A=@n2.R, B=Power(Base=@n3, Exp=@n4)), B=@n5, Radius=@n5, Hardness=Constant())
  n7 = DynamicParameter() {defaultValue=<rgba 1 12 0 1>, paramNames [Exponent, Depth Fade, Gradient Deform, Panning Speed]}
  n8 = Add(A=Multiply(A=Power(Base=@n6, Exp=@n7.Exponent)) {constB=0.5}) {constB=0.25}
  n9 = OneMinus(None=ConstantBiasScale(None=Sine(None=@n8)))
  n10 = Power(Base=@n9)
  n11 = ScalarParameter() {defaultValue=0.3, parameterName=Soft add}
  n12 = ParticleRandom()
  n13 = Multiply(A=TextureCoordinate(), B=AppendVector(A=Constant() {r=1}, B=LinearInterpolate(Alpha=@n12) {constA=0.6, constB=1.5}))
  n14 = Multiply(
   A=Time()
   B=DynamicParameter() {defaultValue=<rgba 1 12 0 1>, paramNames [Exponent, Depth Fade, Gradient Deform, Panning Speed]}.Panning Speed
  )
  n15 = Constant()
  n16 = ParticleRandom()
  n17 = MaterialFunctionCall(XY=Add(A=Panner(Coordinate=@n13, Time=@n14) {speedY=1}, B=AppendVector(A=@n15, B=@n16))) {materialFunction=&Swizzle}
  n18 = Multiply(
   A=@n9
   B=Power(
    Base=TextureSample(UVs=@n17.YX) {texture=/Game/Effects/Fort_Effects/Textures/Smoke/T_Smoke_Trail_Soft_Chromatic}.G
    Exp=ScalarParameter() {defaultValue=3, parameterName=Trail Exponent}
   )
  )
  n19 = ScalarParameter() {defaultValue=1, parameterName=Trail Add}
  n20 = ParticleColor()
  n21 = Multiply(A=@n20.RGB, B=Add(A=Multiply(A=@n10, B=@n11), B=Multiply(A=@n18, B=@n19)))
  MaterialFunctionCall(Emissive=StaticSwitchParameter(True=Multiply(A=@n1.Result, B=@n21), False=@n21) {parameterName=Use TODM Multipliers}) {materialFunction=&MF_FX_EmissiveAdjustments}
  n22 = DynamicParameter() {defaultValue=<rgba 1 12 0 1>, paramNames [Exponent, Depth Fade, Gradient Deform, Panning Speed]}
  Multiply(
   A=@n20.A
   B=ShadingPathSwitch(Default=DepthFade(FadeDistance=@n22.Depth Fade), Mobile=Constant() {r=1}) {desc=Migrated from FeatureLevelSwitch}
  )
 ]
}
