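# Additive unlit material for Niagara sprites that draws a procedural soft round particle without textures.
# A sphere mask over vertically warped UVs forms the shape, with dynamic parameters controlling deformation and falloff exponent, then a sine remap smooths it.
# The shape is multiplied by particle alpha and depth fade, which the mobile path skips.
# Particle color, optionally tinted by time-of-day multipliers through a static switch, feeds a shared emissive adjustment function.
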Material /CRD_Melee/FX/Materials/M_Melee_SoftParticle
graph {
 properties {bUsedWithNiagaraSprites=true, blendMode=BLEND_Additive, pixelDepthOffsetMode=PDOM_Legacy, shadingModel=MSM_Unlit, usedShadingModels=Unlit}
 nodes [
  n1 = MaterialFunctionCall(
   Color_Morning=VectorParameter() {group=TOD, parameterName=Morning, defaultValue=<rgba 1 1 1 1>}.RGB
   Color_Day=VectorParameter() {group=TOD, parameterName=Day, defaultValue=<rgba 1 1 1 1>}.RGB
   Color_Evening=VectorParameter() {group=TOD, parameterName=Evening, defaultValue=<rgba 1 1 1 1>}.RGB
   Color_Night=VectorParameter() {group=TOD, parameterName=Night, defaultValue=<rgba 1 1 1 1>}.RGB
  ) {materialFunction=&Color_Time_of_Day}
  n2 = ParticleColor()
  MaterialFunctionCall(Emissive=StaticSwitchParameter(True=Multiply(A=@n1.Result, B=@n2.RGB), False=@n2.RGB) {parameterName=Use TODM Multipliers}) {materialFunction=&MF_FX_EmissiveAdjustments}
  n3 = MaterialFunctionCall(Float2=TextureCoordinate()) {materialFunction=&BreakOutFloat2Components}
  n4 = OneMinus(None=@n3.G)
  n5 = OneMinus(None=DynamicParameter() {defaultValue=<rgba 1 12 0 0>, paramNames [Exponent, Depth Fade, Gradient Deform, ""]}.Gradient Deform)
  n6 = Constant() {r=0.5}
  n7 = SphereMask(A=AppendVector(A=@n3.R, B=Power(Base=@n4, Exp=@n5)), B=@n6, Radius=@n6, Hardness=Constant())
  n8 = DynamicParameter() {defaultValue=<rgba 1 12 0 0>, paramNames [Exponent, Depth Fade, Gradient Deform, ""]}
  n9 = Add(A=Multiply(A=Power(Base=@n7, Exp=@n8.Exponent)) {constB=0.5}) {constB=0.25}
  n10 = OneMinus(None=ConstantBiasScale(None=Sine(None=@n9)))
  n11 = DynamicParameter() {defaultValue=<rgba 1 12 0 0>, paramNames [Exponent, Depth Fade, Gradient Deform, ""]}
  Multiply(
   A=Multiply(A=@n2.A, B=Power(Base=@n10))
   B=ShadingPathSwitch(Default=DepthFade(FadeDistance=@n11.Depth Fade), Mobile=Constant() {r=1}) {desc=Migrated from FeatureLevelSwitch}
  )
 ]
}
