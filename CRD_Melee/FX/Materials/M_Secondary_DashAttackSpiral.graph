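# Translucent, two-sided unlit flame material for the melee secondary dash attack spiral effect.
# Panning, distorted noise inside a shaped edge-fade mask drives opacity, with fresnel and depth fade skipped on mobile.
# Emissive multiplies fire noise by particle color and time-of-day brightness.
# Switchable world position offset combines SplineThicken, noise waves and a gradient-weighted local offset, with dynamic parameters varying width, UVs and speed.
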Material /CRD_Melee/FX/Materials/M_Secondary_DashAttackSpiral
graph {
 properties {
  bUsedWithClothing=true
  bUsedWithMeshDeformer=true
  bUsedWithNiagaraMeshParticles=true
  bUsedWithSkeletalMesh=true
  bUsedWithSplineMeshes=true
  blendMode=BLEND_Translucent
  pixelDepthOffsetMode=PDOM_Legacy
  refractionMethod=RM_IndexOfRefraction
  shadingModel=MSM_Unlit
  twoSided=true
  usedShadingModels=Unlit
 }
 nodes [
  n1 = ScalarParameter() {defaultValue=1, group=Fire Shape, parameterName=Distortion_UScale, sortPriority=4}
  n2 = ScalarParameter() {defaultValue=1, group=Fire Shape, parameterName=Distortion_VScale, sortPriority=5}
  n3 = StaticSwitchParameter(True=DynamicParameter() {paramNames [UV_Random, SinSpeedRandom, SinDistortIntensity, Param4]}.SinSpeedRandom, False=VertexColor().R) {parameterName=UseParticleUVRandom}
  n4 = Multiply(A=Time(), B=Add(A=ScalarParameter() {defaultValue=1, parameterName=Speed}, B=Multiply(A=@n3) {constB=0.15}))
  n5 = Panner(Coordinate=Multiply(A=TextureCoordinate() {vTiling=1.5}, B=AppendVector(A=@n1, B=@n2)), Time=@n4) {speedY=1}
  n6 = ScalarParameter() {defaultValue=1, group=Fire Shape, parameterName=Fire_MacroShape_UScale}
  n7 = ScalarParameter() {defaultValue=1, group=Fire Shape, parameterName=Fire_MacroShape_VScale, sortPriority=1}
  n8 = MaterialFunctionCall(Float2=Multiply(A=TextureCoordinate(), B=AppendVector(A=@n6, B=@n7))) {materialFunction=&BreakOutFloat2Components}
  n9 = Multiply(A=@n8.G)
  n10 = LinearInterpolate(
   A=ComponentMask(None=TextureSample(UVs=@n5) {samplerType=SAMPLERTYPE_Normal, texture=&T_Weap_Distort_N_02}.RGB) {g=true, r=true}
   Alpha=@n9
  ) {constB=0}
  n11 = ScalarParameter() {defaultValue=3, parameterName=UV_Distortion}
  n12 = ScalarParameter() {defaultValue=1, group=Fire Shape, parameterName=Fire_Micro_UScale, sortPriority=3}
  n13 = ScalarParameter() {defaultValue=1, group=Fire Shape, parameterName=Fire_Micro_VScale, sortPriority=3}
  n14 = MaterialFunctionCall(XY=Multiply(A=TextureCoordinate() {uTiling=0.8, vTiling=0.9}, B=AppendVector(A=@n12, B=@n13))) {materialFunction=&Swizzle}
  n15 = TextureSampleParameter2D(UVs=Add(A=Divide(A=@n10, B=@n11) {constB=20}, B=Panner(Coordinate=@n14.YX, Time=@n4) {speedX=1.2, speedY=0.2})) {parameterName=AlphaTile, samplerType=SAMPLERTYPE_LinearColor, texture=&T_NoisePack_LinearColor}
  n16 = Sine(None=Multiply(A=@n8.R) {constB=0.5})
  n17 = LinearInterpolate(
   A=ScalarParameter() {defaultValue=6, parameterName=Edge Fade Value Top}
   B=ScalarParameter() {defaultValue=2, parameterName=Edge Fade Value Bottom}
   Alpha=@n9
  ) {constA=6, constB=2}
  n18 = Multiply(A=SphereMask(A=TextureCoordinate() {uTiling=1.2}, B=Constant2Vector() {g=0.05, r=0.6}) {attenuationRadius=0.95, hardnessPercent=0}) {constB=15}
  n19 = Multiply(
   A=Multiply(A=Power(Base=@n16, Exp=@n17) {constExponent=4}, B=@n9)
   B=LinearInterpolate(B=Saturate(None=@n18), Alpha=ScalarParameter() {defaultValue=1, parameterName=Bottom Fade}) {constA=1}
  )
  n20 = Multiply(A=Power(Base=Multiply(A=@n15.R, B=@n19), Exp=Constant() {r=1}), B=ScalarParameter() {defaultValue=2, parameterName=OpacityMaskMultiplier}) {constB=3}
  LinearInterpolate(
   A=LinearInterpolate(A=ScalarParameter() {defaultValue=1, parameterName=Norefrac}, B=ScalarParameter() {defaultValue=1, parameterName=refrac}, Alpha=@n20)
   Alpha=MaterialFunctionCall() {materialFunction=&DitherTemporalAA}.Result
  )
  n21 = ScalarParameter() {defaultValue=0.5, parameterName=HDRMult_3}
  n22 = ScalarParameter() {defaultValue=0.8, parameterName=HDRMult_2}
  n23 = ScalarParameter() {defaultValue=0.5, parameterName=HDRMult_1}
  n24 = ScalarParameter() {defaultValue=0.35, parameterName=HDRMult}
  n25 = ParticleColor()
  n26 = DynamicParameter() {paramNames [UV_Random, SinSpeedRandom, SinDistortIntensity, Param4]}
  n27 = Multiply(
   A=TextureCoordinate() {uTiling=0.3, vTiling=0.7}
   B=AppendVector(
    A=ScalarParameter() {defaultValue=1, group=Color, parameterName=Fire_Color_UScale}
    B=ScalarParameter() {defaultValue=1, group=Color, parameterName=Fire_Color_VScale, sortPriority=1}
   )
  )
  n28 = TextureSampleParameter2D(UVs=Panner(Coordinate=Add(A=@n26.UV_Random, B=@n27), Time=@n4) {speedY=0.6}) {parameterName=EmissiveTile, texture=/Game/Global/Textures/Default/Noise/T_FireNoise_02}
  MaterialFunctionCall(Emissive=Multiply(
    A=MaterialFunctionCall(Color_Morning=@n21, Color_Day=@n22, Color_Evening=@n23, Color_Night=@n24) {materialFunction=&Color_Time_of_Day}.Result
    B=Multiply(A=@n25.RGB, B=@n28.G)
   )) {materialFunction=&MF_FX_EmissiveAdjustments}
  n29 = ScalarParameter() {defaultValue=2, parameterName=Fresnel-Hardness}
  n30 = ScalarParameter() {defaultValue=2, parameterName=Fresnel-Exponent}
  n31 = DynamicParameter() {parameterIndex=1, paramNames [WidthBase, WidthTip, SinDistortIntensity, Param4]}
  n32 = Multiply(A=VertexColor().R) {constB=0.5}
  n33 = ScalarParameter() {defaultValue=35, parameterName=WidthBase}
  n34 = ScalarParameter() {defaultValue=100, parameterName=WidthTip}
  n35 = MaterialFunctionCall(
   WidthBase=StaticSwitchParameter(True=Add(A=@n31.WidthBase, B=@n32), False=Add(A=@n33, B=@n32)) {defaultValue=true, parameterName=UseDynaParamterForWidth?}
   WidthTip=StaticSwitchParameter(True=Add(A=@n31.WidthTip, B=@n32), False=Add(A=@n34, B=@n32)) {defaultValue=true, parameterName=UseDynaParamterForWidth?}
  ) {materialFunction=&SplineThicken}
  n36 = DepthFade(
   Opacity=Multiply(
    A=@n20
    B=MaterialFunctionCall(Brightness=@n29, Exponent=@n30, Vertex / Pixel Normal=@n35.Normal) {materialFunction=&MF_Fresnel_FX}.Fresnel
   )
   FadeDistance=ScalarParameter() {defaultValue=24, parameterName=DepthFadeDistance}
  )
  Multiply(A=ShadingPathSwitch(Default=@n36, Deferred=@n36, Mobile=@n20) {desc=Migrated from FeatureLevelSwitch}, B=ParticleColor().A)
  n37 = Panner() {speedX=-1, speedY=1}
  Normalize(VectorInput=Add(
    A=Constant3Vector() {constant=<rgba 1 -1 0 0>}
    B=TextureSample(UVs=@n37) {samplerType=SAMPLERTYPE_Normal, texture=/Game/Effects/Fort_Effects/Textures/Tile/noise_flamy_Tile_NRM}.RGB
   ))
  n38 = StaticSwitchParameter(True=@n35.WorldPositionOffset, False=Constant()) {defaultValue=true, parameterName=Use Spline Thicken}
  n39 = Panner(Coordinate=TextureCoordinate(), Time=@n4) {speedY=0.8}
  n40 = ComponentMask(None=TextureSample(UVs=@n39) {samplerType=SAMPLERTYPE_Normal, texture=&T_Weap_Distort_N_02}.RGB) {g=true, r=true}
  n41 = TextureSample(UVs=@n39) {texture=&T_Smooth_tile}
  n42 = Multiply(A=MaterialFunctionCall(Float2=TextureCoordinate()) {materialFunction=&BreakOutFloat2Components}.G)
  n43 = Multiply(
   A=LinearInterpolate(B=AppendVector(A=@n40, B=@n41.R), Alpha=OneMinus(None=@n42))
   B=ScalarParameter() {defaultValue=10, parameterName=WaveIntensity}
  ) {constB=10}
  n44 = Transform(None=VectorParameter() {parameterName=LocalPostioning}.RGB) {transformSourceType=TRANSFORMSOURCE_Local}
  n45 = TextureCoordinate()
  n46 = Power(
   Base=OneMinus(None=ComponentMask(None=@n45) {g=true})
   Exp=ScalarParameter() {defaultValue=2, parameterName=Edge Gradient Movement Control Exp}
  )
  StaticSwitchParameter(True=Constant(), False=Add(A=Add(A=@n38, B=@n43), B=Multiply(A=@n44, B=@n46))) {parameterName=TurnOffWPO}
 ]
}
